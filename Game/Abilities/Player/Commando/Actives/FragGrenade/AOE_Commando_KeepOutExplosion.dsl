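; Area-of-effect actor for the Commando frag grenade's keep-out explosion, with a damage sphere, lingering particles and looping-style area audio; no damage logic appears here.
; Send_Info sets Radius and resizes the DamageArea sphere, which BeginPlay registers as the collision component.
; OnRep_Radius scales the particles, stores rebuilt attenuation settings with radius as falloff distance (never applied here), and plays the audio.
; Destruction triggers an audio fade-out.
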(event EventDestroyed
  (Audio|Components|Audio|FadeOut (Variables|Default|GetAOE_Audio_Effect) 1 0.0 "Linear" :!class /Script/Engine.AudioComponent))

(event Custom|Send_Info
  (params (Radius :type Float%20%28double-precision%29))
  (bind _output_get (Variables|Default|SetRadius Radius))
  (Components|Sphere|SetSphereRadius (Variables|Default|GetDamageArea) _output_get false))

(event EventBeginPlay
  (Collision|SetCollisionComponent (Variables|Default|GetDamageArea)))

(fn OnRep_Radius
  (bind _p_grenade_linger (Variables|Default|GetP_Grenade_Linger))
  (bind _radius (Variables|Default|GetRadius))
  (Effects|Components|ParticleSystem|SetFloatParameter _p_grenade_linger "size" _radius)
  (Effects|Components|ParticleSystem|SetVectorParameter _p_grenade_linger "size" (Math|Vector|MakeVector _radius _radius _radius))
  (bind _outattenuationsettings (Audio|Components|Audio|GetAttenuationSettingsToApply (Variables|Default|GetAOE_Audio_Effect)))
  (Variables|Default|SetAOE_Sound_Attenuation_Based_On_Radius (Utilities|Struct|MakeSoundAttenuationSettings :bAttenuate (Utilities|Struct|BreakSoundAttenuationSettings _outattenuationsettings) :bSpatialize (.bSpatialize (Utilities|Struct|BreakSoundAttenuationSettings _outattenuationsettings)) :bAttenuateWithLPF (.bAttenuateWithLPF (Utilities|Struct|BreakSoundAttenuationSettings _outattenuationsettings)) :bEnableListenerFocus (.bEnableListenerFocus (Utilities|Struct|BreakSoundAttenuationSettings _outattenuationsettings)) :bEnableFocusInterpolation false :bEnableOcclusion (.bEnableOcclusion (Utilities|Struct|BreakSoundAttenuationSettings _outattenuationsettings)) :bUseComplexCollisionForOcclusion (.bUseComplexCollisionForOcclusion (Utilities|Struct|BreakSoundAttenuationSettings _outattenuationsettings)) :bEnableReverbSend true :bEnablePriorityAttenuation false :bApplyNormalizationToStereoSounds false :bEnableLogFrequencyScaling false :bEnableSubmixSends false :bEnableSourceDataOverride false :bEnableSendToAudioLink true :SpatializationAlgorithm (.SpatializationAlgorithm (Utilities|Struct|BreakSoundAttenuationSettings _outattenuationsettings)) :BinauralRadius 0.0 :CustomLowpassAirAbsorptionCurve "(EditorCurveData=(Keys=,DefaultValue=340282346638528859811704183484516925440.000000,PreInfinityExtrap=RCCE_Constant,PostInfinityExtrap=RCCE_Constant),ExternalCurve=None)" :CustomHighpassAirAbsorptionCurve "(EditorCurveData=(Keys=,DefaultValue=340282346638528859811704183484516925440.000000,PreInfinityExtrap=RCCE_Constant,PostInfinityExtrap=RCCE_Constant),ExternalCurve=None)" :AbsorptionMethod "Linear" :OcclusionTraceChannel "ECC_Visibility" :ReverbSendMethod "Linear" :PriorityAttenuationMethod "Linear" :NonSpatializedRadiusStart (.NonSpatializedRadiusStart (Utilities|Struct|BreakSoundAttenuationSettings _outattenuationsettings)) :NonSpatializedRadiusEnd 0.0 :NonSpatializedRadiusMode "OmniDirectional" :StereoSpread (.StereoSpread (Utilities|Struct|BreakSoundAttenuationSettings _outattenuationsettings)) :LPFRadiusMin (.LPFRadiusMin (Utilities|Struct|BreakSoundAttenuationSettings _outattenuationsettings)) :LPFRadiusMax (.LPFRadiusMax (Utilities|Struct|BreakSoundAttenuationSettings _outattenuationsettings)) :LPFFrequencyAtMin 2e+04 :LPFFrequencyAtMax 2e+04 :HPFFrequencyAtMin 0.0 :HPFFrequencyAtMax 0.0 :FocusAzimuth 3e+01 :NonFocusAzimuth 6e+01 :FocusDistanceScale 1.0 :NonFocusDistanceScale 1.0 :FocusPriorityScale 1.0 :NonFocusPriorityScale 1.0 :FocusVolumeAttenuation 1.0 :NonFocusVolumeAttenuation 1.0 :FocusAttackInterpSpeed 1.0 :FocusReleaseInterpSpeed 1.0 :OcclusionLowPassFilterFrequency 2e+04 :OcclusionVolumeAttenuation 1.0 :OcclusionInterpolationTime 0.1 :ReverbWetLevelMin 0.3 :ReverbWetLevelMax 0.95 :ReverbDistanceMin 4e+02 :ReverbDistanceMax 4e+03 :ManualReverbSendLevel 0.0 :PriorityAttenuationMin 1.0 :PriorityAttenuationMax 1.0 :PriorityAttenuationDistanceMin 4e+02 :PriorityAttenuationDistanceMax 4e+03 :ManualPriorityAttenuation 1.0 :CustomReverbSendCurve "(EditorCurveData=(Keys=,DefaultValue=340282346638528859811704183484516925440.000000,PreInfinityExtrap=RCCE_Constant,PostInfinityExtrap=RCCE_Constant),ExternalCurve=None)" :CustomPriorityAttenuationCurve "(EditorCurveData=(Keys=,DefaultValue=340282346638528859811704183484516925440.000000,PreInfinityExtrap=RCCE_Constant,PostInfinityExtrap=RCCE_Constant),ExternalCurve=None)" :PluginSettings "(SpatializationPluginSettingsArray=,OcclusionPluginSettingsArray=,ReverbPluginSettingsArray=,SourceDataOverridePluginSettingsArray=)" :DistanceAlgorithm (.DistanceAlgorithm (Utilities|Struct|BreakSoundAttenuationSettings _outattenuationsettings)) :AttenuationShape (.AttenuationShape (Utilities|Struct|BreakSoundAttenuationSettings _outattenuationsettings)) :FalloffMode "Continues" :dBAttenuationAtMax (.dBAttenuationAtMax (Utilities|Struct|BreakSoundAttenuationSettings _outattenuationsettings)) :AttenuationShapeExtents (.AttenuationShapeExtents (Utilities|Struct|BreakSoundAttenuationSettings _outattenuationsettings)) :ConeOffset (.ConeOffset (Utilities|Struct|BreakSoundAttenuationSettings _outattenuationsettings)) :FalloffDistance _radius :ConeSphereRadius 0.0 :ConeSphereFalloffDistance 0.0 :CustomAttenuationCurve (.CustomAttenuationCurve (Utilities|Struct|BreakSoundAttenuationSettings _outattenuationsettings))))
  (Audio|Components|Audio|Play (Variables|Default|GetAOE_Audio_Effect) 0.0))

(fn ConstructionScript)
