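; Commando frag grenade projectile that bounces, explodes and can spawn perk follow-ups.
; Pawn overlaps fake a bounce by reversing velocity; bounce count, stopping, damage, water or tagged overlaps trigger detonation, mostly after short delays.
; On explosion, authority sends the instigator a gameplay event and handles cluster grenades and Keep Out clouds, while remotes play effects, sound, shake and force feedback.
; Destruction also forces explosion effects, covering kill-z deaths.
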(event Collision|EventActorBeginOverlap
  (params (OtherActor :type Actor%20Object%20Reference))
  ;[1 1 1] If we overlap a pawn, fake a bounce.
  (bind _aspawn (Utilities|Casting|CastToPawn OtherActor)
    (:then
      (if (Utilities|NotEqual(Object) _aspawn (Game|GetInstigator))
        (Variables|/Script/Engine.ProjectileMovementComponent|SetProjectileGravityScale (Variables|Default|GetBouncePawnAgainstPawnGravityScale) (Variables|Movement|GetProjectileMovementComponent)) ; Hit a pawn, scale gravity
        (Variables|/Script/Engine.MovementComponent|SetVelocity (* (Math|Vector|MakeVector (.x (* (Variables|/Script/Engine.MovementComponent|GetVelocity (Variables|Movement|GetProjectileMovementComponent)) -1)) (.y (* (Variables|/Script/Engine.MovementComponent|GetVelocity (Variables|Movement|GetProjectileMovementComponent)) -1)) (Math|Float|Absolute(Float) (.z (* (Variables|/Script/Engine.MovementComponent|GetVelocity (Variables|Movement|GetProjectileMovementComponent)) -1)))) (Variables|/Script/Engine.ProjectileMovementComponent|GetBounciness (Variables|Movement|GetProjectileMovementComponent))) (Variables|Movement|GetProjectileMovementComponent))
        (Variables|Default|SetCurrentNumberOfBounces (+ (Variables|Default|GetCurrentNumberOfBounces) 1))
        (if (>= (Variables|Default|GetCurrentNumberOfBounces) (Variables|Default|GetNumberOfBouncesTillExplode))
          (Utilities|FlowControl|DoOnce false
            (:Completed
              ;[0.7116 0.838 1] OnStop - Distance Guide
              (Utilities|FlowControl|Delay 0.3 :!pinned true) ; Time Before Explosion
              (Utilities|FlowControl|DoOnce false
                (:Completed
                  (CallFunction|UnbindDestroyGrenade)
                  (CallFunction|TriggerDoExplsoion))))))))
    (:CastFailed
      (bind _returnvalue (GameplayTags|HasAnyMatchingGameplayTags(Message) OtherActor (Variables|Default|GetTC_ActorTagsThatShouldExplodeOnOverlap)))
      (if _returnvalue
        (CallFunction|TriggerDoExplsoion)))))

(event EventOnBounce
  (params (Hit :type Hit%20Result%20Structure%20%28by%20ref%29))
  (bind (_bblockinghit _binitialoverlap _time _distance _location _impactpoint _normal _impactnormal _physmat _hitactor) (Collision|BreakHitResult Hit))
  ;[1 1 1] OnBounce - After X bounces set to explode
  (bind _aspawn (Utilities|Casting|CastToPawn _hitactor)
    (:then
      (if (Utilities|NotEqual(Object) _aspawn (Game|GetInstigator))
        (Variables|/Script/Engine.ProjectileMovementComponent|SetProjectileGravityScale (Variables|Default|GetBouncePawnAgainstPawnGravityScale) (Variables|Movement|GetProjectileMovementComponent)) ; Hit a pawn, scale gravity
      ))
    (:CastFailed
      (if (Utilities|Equal(Object) _physmat "/Game/FortressPhysicalMaterials/Water.Water")
        (CallFunction|TriggerDoExplsoion)
        (else
          _))))
  (Variables|Default|SetCurrentNumberOfBounces (+ (Variables|Default|GetCurrentNumberOfBounces) 1))
  (if (>= (Variables|Default|GetCurrentNumberOfBounces) (Variables|Default|GetNumberOfBouncesTillExplode))
    (Utilities|FlowControl|DoOnce false
      (:Completed
        ;[0.7116 0.838 1] OnStop - Distance Guide
        (Utilities|FlowControl|Delay 0.3 :!pinned true) ; Time Before Explosion
        (Utilities|FlowControl|DoOnce false
          (:Completed
            (CallFunction|UnbindDestroyGrenade)
            (CallFunction|TriggerDoExplsoion)))))))

(event Custom|TriggerDoExplsoion
  ;[1 1 1] FORT-59119
  (Utilities|FlowControl|DoOnce false
    (:Completed
      (Delivery|DoExplosion))))

(event EventBeginPlay
  ;[1 1 1] BeginPlay - Fuse Sound
  (CallFunction|BindDestroyGrenade)
  (Networking|CanExecuteCosmeticEvents
    (:True
      (Audio|Components|Audio|SetSound (Variables|B_Prj_Grenade|GetGrenadeFuse_AudioComponent) (Variables|Default|GetCue_GrenadeFuseSound)))))

(event EventOnStop
  (params (Hit :type Hit%20Result%20Structure%20%28by%20ref%29))
  (bind _effect_distance (Variables|B_Prj_Grenade|GetEffect_Distance))
  ;[0.7116 0.838 1] OnStop - Distance Guide
  (CallFunction|StopRotation)
  (Transformation|SetWorldRotation _effect_distance "90.000000,0.000000, 0" false false)
  (Components|Activation|Activate _effect_distance false)
  (Effects|Components|ParticleSystem|SetVectorParameter _effect_distance "Size" (Math|Conversions|ToVector(Float) (Variables|Default|GetExplosionRadius)))
  (if (not (Team|OnSameTeam (Game|GetInstigatorController) (Game|GetPlayerController 0)))
    (Effects|Components|ParticleSystem|SetVectorParameter _effect_distance "Color" "1.500000,0.110000,0.190000"))
  (Utilities|FlowControl|Delay 0.3 :!pinned true) ; Time Before Explosion
  (Utilities|FlowControl|DoOnce false
    (:Completed
      (CallFunction|UnbindDestroyGrenade)
      (CallFunction|TriggerDoExplsoion))))

(event Game|Damage|EventAnyDamage
  (params (Damage :type Float%20%28single-precision%29) (DamageType :type Damage%20Type%20Object%20Reference) (InstigatedBy :type Controller%20Object%20Reference) (DamageCauser :type Actor%20Object%20Reference))
  ;[0.7116 0.838 1] OnStop - Distance Guide
  (Utilities|FlowControl|Delay 0.1)
  (Utilities|FlowControl|DoOnce false
    (:Completed
      (CallFunction|UnbindDestroyGrenade)
      (CallFunction|TriggerDoExplsoion))))

(event Custom|Stop_Rotation
  ;[1 1 1] Stop Rotating Movement
  (Components|Movement|SetUpdatedComponent (Variables|Default|GetRotatingMovement) 0))

(event EventDestroyed
  (CallFunction|ForceonExplodedEffects)
  (CallFunction|UnbindDestroyGrenade))

(event Custom|Bind%20Destroy%20Grenade)

(event EventOnExploded
  (params (HitActors :type Array%20of%20Actor%20Object%20References) (HitResults :type Array%20of%20Hit%20Result%20Structures))
  (bind _returnvalue (Transformation|GetActorLocation))
  ;[1 1 1] OnExploded - Emitter, Sound, Camera Shake
  (Utilities|FlowControl|DoOnce false
    (:Completed
      (Networking|SwitchHasAuthority
        (:Authority
          ;;[1 1 1] Send explosion event to trigger other abilities.
          (CallFunction|SendExplodedEvent)
          ;;[1 0.9225 0] Handle Keep out HEROES 1.0
          (CallFunction|CheckKeepOutandCluster))
        (:Remote
          (Networking|CanExecuteCosmeticEvents
            (:True
              (Effects|SpawnEmitteratLocation (Variables|Default|GetP_Explosion) (Transformation|GetActorLocation) (Math|Rotator|MakeRotator 0.0 90))
              (Utilities|FlowControl|Sequence
                (:then_0
                  (Camera|PlayWorldCameraShake "/Game/Abilities/Player/Commando/Actives/FragGrenade/B_GrenadeCameraShake.B_GrenadeCameraShake_C" (Transformation|GetActorLocation) 200 800))
                (:then_1
                  (if (Pawn|IsLocallyControlled (Game|GetInstigator))
                    (Audio|PlaySoundatLocation (Variables|Sounds|GetCue_CloseSound) _returnvalue) ; Play Close Explosion
                    (Audio|Components|Audio|FadeOut (Variables|B_Prj_Grenade|GetGrenadeFuse_AudioComponent) 0.2 0.0 "Linear" :!class /Script/Engine.AudioComponent) ; Fade Out Fuse Sound
                    (else
                      (Audio|PlaySoundatLocation (Variables|Sounds|GetCue_DistantSound) _returnvalue) ; Play Distant Explosion
                    )))
                (:then_2
                  (Fortnite|PlayLocalForceFeedbackatLocationMulti (Variables|Default|GetExplosionForceFeedbackNear) (Variables|Default|GetExplosionForceFeedbackFar) (Transformation|GetActorLocation) (Variables|Default|GetExplosionRadius) (* (Variables|Default|GetExplosionRadius) 3) "Explosion")
                  (Development|SetHiddeninGame (Variables|B_Prj_Grenade|GetMesh) true false))))))))))

(event Custom|On%20Destroy%20Grenade
  (params (DestroyedActor :type Actor%20Object%20Reference))
  ;[1 1 1] FORT-48686
  ; Allows explosion effects to play when destroyed by kill-z
  (CallFunction|ForceonExplodedEffects)
  (CallFunction|UnbindDestroyGrenade))

(event Custom|Force%20On%20Exploded%20Effects
  (bind _returnvalue (Transformation|GetActorLocation))
  ;[1 1 1] OnExploded - Emitter, Sound, Camera Shake
  (if (not (Variables|Rendering|GetbHidden))
    (Utilities|FlowControl|DoOnce false
      (:Completed
        (Networking|SwitchHasAuthority
          (:Authority
            ;;[1 1 1] Send explosion event to trigger other abilities.
            (CallFunction|SendExplodedEvent)
            ;;[1 0.9225 0] Handle Keep out HEROES 1.0
            (CallFunction|CheckKeepOutandCluster))
          (:Remote
            (Networking|CanExecuteCosmeticEvents
              (:True
                (Effects|SpawnEmitteratLocation (Variables|Default|GetP_Explosion) (Transformation|GetActorLocation) (Math|Rotator|MakeRotator 0.0 90))
                (Utilities|FlowControl|Sequence
                  (:then_0
                    (Camera|PlayWorldCameraShake "/Game/Abilities/Player/Commando/Actives/FragGrenade/B_GrenadeCameraShake.B_GrenadeCameraShake_C" (Transformation|GetActorLocation) 200 800))
                  (:then_1
                    (if (Pawn|IsLocallyControlled (Game|GetInstigator))
                      (Audio|PlaySoundatLocation (Variables|Sounds|GetCue_CloseSound) _returnvalue) ; Play Close Explosion
                      (Audio|Components|Audio|FadeOut (Variables|B_Prj_Grenade|GetGrenadeFuse_AudioComponent) 0.2 0.0 "Linear" :!class /Script/Engine.AudioComponent) ; Fade Out Fuse Sound
                      (else
                        (Audio|PlaySoundatLocation (Variables|Sounds|GetCue_DistantSound) _returnvalue) ; Play Distant Explosion
                      )))
                  (:then_2
                    (Fortnite|PlayLocalForceFeedbackatLocationMulti (Variables|Default|GetExplosionForceFeedbackNear) (Variables|Default|GetExplosionForceFeedbackFar) (Transformation|GetActorLocation) (Variables|Default|GetExplosionRadius) (* (Variables|Default|GetExplosionRadius) 3) "Explosion")
                    (Development|SetHiddeninGame (Variables|B_Prj_Grenade|GetMesh) true false)))))))))))

(event Custom|Unbind%20Destroy%20Grenade
  ;[1 1 1] FORT-48686
  ; Allows explosion effects to play when destroyed by kill-z
  (Game|UnbindEventfromOnDestroyed (delegate Custom|On%20Destroy%20Grenade)))

(event Custom|CheckKeepOutAndCluster
  ;[1 1 1] Spaws Cluster or Keep out, if unlocked
  (Utilities|FlowControl|Sequence
    (:then_0
      (if (Variables|Default|GetbHasCluster)
        _
        (elif (Variables|Default|GetbHasClusterTactical)
          _))
      (CallFunction|SpawnClusters))
    (:then_1
      (if (Variables|Default|GetbHasKeepOut)
        (CallFunction|SpawnKeepOut)))))

(event Custom|SendExplodedEvent
  (bind _returnvalue (Game|GetInstigator))
  ;[1 1 1] Send Gameplay Event back to actor who threw the grenade so that perks can be triggered off of it. (Examples, Keep Out or Cluster Bomb). IN OTHER BP ALWAYS MAKE SURE GRENADE IS VALID AS THIS BP MAY KILL IT OFF.
  (Utilities|IsValid _returnvalue
    (:Is%20Valid
      (Ability|SendGameplayEventtoActor _returnvalue (Variables|Default|GetT_Event_GrenadeExploded) (Utilities|Struct|MakeGameplayEventData (Variables|Default|GetT_Event_GrenadeExploded) self _returnvalue 0 0 "()" "(GameplayTags=)" "(GameplayTags=)" 0.0 (Ability|TargetData|AbilityTargetDatafromLocations (Utilities|Struct|MakeGameplayAbilityTargetingLocationInfo self 0 0 (Transformation|GetActorTransform self)) (Utilities|Struct|MakeGameplayAbilityTargetingLocationInfo self 0 0 (Transformation|GetActorTransform self))))))))

(fn CalcGrenadeSpeed
  (params (Angle :type Float%20%28double-precision%29))
  (return (Math|Float|Lerp (LocalVariables|GetGrenadeSpeedMin) (LocalVariables|GetGrenadeSpeedMax) (/ (- (Math|Float|Clamp(Float) (Math|Rotator|NormalizeAxis Angle) (LocalVariables|GetMinSpeedPitch) (LocalVariables|GetMaxSpeedPitch)) (LocalVariables|GetMinSpeedPitch)) (- (LocalVariables|GetMaxSpeedPitch) (LocalVariables|GetMinSpeedPitch))))))

(fn OnRep_StoredHit)

(fn SpawnClusters
  (for _ (range 1 (+ (Variables|Default|GetMaxClusterGrenades) 1))
    (Projectile|SpawnProjectile :ProjectileClass (Variables|Default|GetPrj_Cluster) :RequestedBy (Actor|GetOwner) :SpawnLocation (+ (Transformation|GetActorLocation) "0, 0,15.000000") :SpawnRotation (Math|Rotator|MakeRotator 0.0 (Math|Random|RandomFloatinRange -60 100) (Math|Random|RandomFloatinRange -360 360)) :SpawnDirection (Math|Rotator|MakeRotator 0.0 (Math|Random|RandomFloatinRange -60 100) (Math|Random|RandomFloatinRange -360 360)) :EffectContainerSpecToApplyOnExplode (select (Variables|Default|GetbHasCluster) (Variables|Default|GetClusterContainerSpec) (Variables|Default|GetCluster_Tactical_Container_Spec)) :bUseDefaultPhysics false :InitialSpeed (Math|Random|RandomFloatinRange 700 1400) :GravityScale 1.0 :ChargePercent 0.0 :!pinned true) ; Cluster Grenades
  ))

(fn SpawnKeepOut
  (bind _returnvalue (AreaOfEffect|SpawnAreaOfEffectCloud (Variables|Default|GetAOE_KeepOut) (Actor|GetOwner) (Transformation|GetActorLocation) (Math|Rotator|MakeRotator 0.0 9e+01) (Variables|Default|GetKeepOutContainerSpec) :!pinned true)) ; Grenade AOE Cloud
  (bind _asaoe_commando_keep_out_explosion (Utilities|Casting|CastToAOE_Commando_KeepOutExplosion _returnvalue))
  ;[1 0.2004 0] Hook up with radius
  (Class|AOECommandoKeepOutExplosion|SendInfo _asaoe_commando_keep_out_explosion (Variables|Default|GetExplosionRadius)))

(fn ConstructionScript
  (Networking|CanExecuteCosmeticEvents
    (:True
      (Components|Movement|SetUpdatedComponent (Variables|Default|GetRotatingMovement) (Variables|B_Prj_Grenade|GetMesh)))))
